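# Material function for the RL251SE wheel that layers an animated aurora night-sky look over MF_WheelParent_Simple, overriding emissive and setting base color to black.
# Rim emissive combines a sine-driven mask pulse with accent tinting, then blends against the tire effect.
# Aurora uses panned clouds blending custom and accent colors, masked through MF_Infinity_OffsetFunction.
# Named reroutes carry fresnel, bump-offset sky gradient, noise-modulated stars and time-staggered comets from RGBA-packed streaks.
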MaterialFunction /VehicleCosmetics/Wheels/RL251SE/Materials/MF_aurora
graph {nodes [
  n1 = NamedRerouteUsage()
  n2 = ScalarParameter() {defaultValue=0.5, parameterName=Rim_pulse_brightness}
  n3 = TextureSampleParameter2D() {parameterName=RimMaskTexture, texture=/VehicleCosmetics/Wheels/RL251SE/Textures/T_RL251SE_mask}
  n4 = Multiply(A=Time(), B=ScalarParameter() {defaultValue=0.2, parameterName=Rim pulse speed})
  n5 = Abs(None=Add(A=Sine(None=@n4)))
  n6 = Add(A=Multiply(A=@n3.R, B=@n5), B=Divide(A=Add(A=@n3.R, B=@n5)) {constB=5})
  n7 = VectorParameter() {parameterName=AccentColor, defaultValue=<rgba 0.4233 0.0782 0.8469 1>}
  n8 = SphereMask(A=Subtract(A=ObjectPositionWS(), B=WorldPosition().XYZ), B=Constant3Vector()) {attenuationRadius=50, hardnessPercent=60}
  n9 = LinearInterpolate(B=Desaturation(None=Multiply(A=@n7.RGB) {constB=0.6}, Fraction=Constant() {r=0.2}), Alpha=Power(Base=OneMinus(None=@n8)))
  n10 = NamedRerouteUsage()
  n11 = LinearInterpolate(
   A=LinearInterpolate(A=NamedRerouteUsage(), B=Multiply(A=@n1, B=@n2), Alpha=Clamp(None=@n6))
   B=Multiply(A=Add(A=@n9, B=@n10), B=NamedRerouteUsage())
   Alpha=OneMinus(None=TextureSample() {samplerType=SAMPLERTYPE_LinearColor, texture=/VehicleCosmetics/Wheels/RL251/Textures/T_Rim_RL251_M}.R)
  )
  n12 = ComponentMask(None=MaterialFunctionCall!(
    Texture=TextureObject() {texture=/VehicleCosmetics/Wheels/RL251SE/Textures/T_RL251SE_mask_tire}
    Offset=Constant() {r=-0.2}
    RotateUVs
    GradientDarkness=Constant() {r=0.002}
    GradientFalloff
    UVs
    Opt_LowQualityOffset
    Opt_MediumQualityOffset
   ) {materialFunction=/VehicleCosmetics/SharedMaterials/MaterialFunctions/MF_Infinity_OffsetFunction}.Result) {r=true}
  n13 = Multiply(A=OneMinus(None=NamedRerouteUsage()), B=Constant() {r=0.075})
  n14 = VectorParameter() {parameterName=CustomColor, defaultValue=<rgba 0.016 1 0.4621 1>}
  n15 = Panner(Coordinate=TextureCoordinate() {uTiling=0.2, vTiling=1.24}) {speedX=0.145, speedY=0.1}
  n16 = Panner(Coordinate=TextureCoordinate()) {speedX=-0.23, speedY=-0.18}
  n17 = ComponentMask(None=TextureCoordinate() {uTiling=2}) {g=true}
  n18 = Multiply(
   A=Add(
    A=TextureSample(UVs=@n15) {samplerType=SAMPLERTYPE_Grayscale, texture=/VehicleCosmetics/Wheels/RL273SE/Textures/T_Wheel_RL273SE_clouds}.R
    B=TextureSample(UVs=@n16) {samplerType=SAMPLERTYPE_Grayscale, texture=&T_Wheel_RL273SE_clouds}.R
   )
   B=Multiply(A=@n17, B=Add(A=@n17, B=@n17))
  )
  n19 = ScalarParameter() {defaultValue=0.4, parameterName=Aurora_offset}
  n20 = LinearInterpolate(
   A=@n14.RGB
   B=VectorParameter() {parameterName=AccentColor, defaultValue=<rgba 0.4233 0.0782 0.8469 1>}.RGB
   Alpha=Clamp(None=Add(A=@n18, B=@n19))
  )
  n21 = Clamp(None=Add(A=@n20, B=Multiply(A=@n20, B=@n20)))
  n22 = ScalarParameter() {defaultValue=1, parameterName=PaintedColorOverride}
  n23 = Add(A=NamedRerouteUsage(), B=Multiply(A=Add(A=@n12, B=@n13), B=LinearInterpolate(A=@n21, B=@n20, Alpha=@n22)))
  n24 = NamedRerouteUsage()
  FunctionOutput(None=SetMaterialAttributes!(
    MaterialAttributes=MaterialFunctionCall() {materialFunction=&MF_WheelParent_Simple}.MaterialAttributes
    EmissiveColor=LinearInterpolate(A=@n11, B=@n23, Alpha=@n24)
    BaseColor=Constant()
   ) {attributeSetTypes [#1, #2]})
  n25 = Fresnel() {exponent=8}
  n26 = Power(Base=OneMinus(None=Fresnel() {exponent=2})) {constExponent=10}
  n27 = Multiply(A=ScalarParameter() {defaultValue=3, parameterName=FresnelStrength}, B=Add(A=Power(Base=@n25) {constExponent=0.7}, B=Clamp(None=@n26)))
  NamedRerouteDeclaration(None=Clamp(None=OneMinus(None=@n27))) {name=Fresnel, nodeColor=<rgba 0.3594 1 0 1>}
  NamedRerouteDeclaration(None=@n14.RGB) {name=CustomColor, nodeColor=<rgba 1 0 0.0469 1>}
  n28 = VectorParameter() {parameterName=Sky_color_dark, defaultValue=<rgba 0.0068 0.0147 0.0469 1>}
  NamedRerouteDeclaration(None=Multiply(A=ScalarParameter() {defaultValue=2, parameterName=SkyDark_intensity}, B=@n28.RGB)) {name=sky_dark, nodeColor=<rgba 0.5469 1 0 1>}
  n29 = Multiply(B=TextureCoordinate() {uTiling=0.5, vTiling=2}) {constA=0.5}
  n30 = Multiply(A=Time(), B=ScalarParameter() {defaultValue=0.2, parameterName=CometSpeed})
  n31 = Constant() {r=-0.5}
  n32 = TextureSample(UVs=BumpOffset(Coordinate=Panner(Coordinate=@n29, Time=@n30) {speedX=1, speedY=-1}, Height=@n31)) {texture=/VehicleCosmetics/Wheels/RL251SE/Textures/T_RL251SE_Space_Streaks_RGB}
  n33 = ScalarParameter() {defaultValue=0.5, parameterName=CometTimeOffset}
  n34 = ScalarParameter() {defaultValue=1.3, parameterName=TimeOffsetMultiply}
  n35 = Multiply(A=Add(A=@n30, B=@n33) {desc=1}, B=@n34)
  n36 = TextureSample(UVs=BumpOffset(Coordinate=Panner(Coordinate=@n29, Time=@n35) {speedX=1, speedY=-1}, Height=@n31)) {texture=&T_RL251SE_Space_Streaks_RGB}
  n37 = Multiply(A=Add(A=@n35, B=@n33) {desc=2}, B=@n34)
  n38 = TextureSample(UVs=BumpOffset(Coordinate=Panner(Coordinate=@n29, Time=@n37) {speedX=1, speedY=-1}, Height=@n31)) {texture=&T_RL251SE_Space_Streaks_RGB}
  n39 = Multiply(A=Add(A=@n37, B=@n33) {desc=3}, B=@n34)
  n40 = TextureSample(UVs=BumpOffset(Coordinate=Panner(Coordinate=@n29, Time=@n39) {speedX=1, speedY=-1}, Height=@n31)) {texture=&T_RL251SE_Space_Streaks_RGB}
  n41 = Multiply(A=Add(A=Add(A=@n32.R, B=@n36.G), B=Add(A=@n38.B, B=@n40.A)))
  NamedRerouteDeclaration(None=@n41) {name=comets, nodeColor=<rgba 0 0.3672 1 1>}
  n42 = TextureCoordinate()
  n43 = ScalarParameter() {defaultValue=-10, parameterName=SkyOffset}
  n44 = TextureSample(UVs=BumpOffset(Coordinate=@n42, Height=@n43)) {texture=/VehicleCosmetics/Wheels/RL251SE/Textures/T_Negative_GradientRGB1}
  NamedRerouteDeclaration(None=@n44.G) {name=TireGradientMask, nodeColor=<rgba 0.3359 0 1 1>}
  n45 = TextureSample(UVs=Panner() {speedX=0.1, speedY=0.1}) {texture=/VehicleCosmetics/Wheels/RL251SE/Textures/T_UVCurlNoise}
  n46 = ScalarParameter() {defaultValue=0.5, parameterName=StarsTiling}
  n47 = Panner(Coordinate=BumpOffset(Coordinate=Multiply(A=@n42, B=@n46), Height=@n43)) {speedY=-0.015}
  n48 = Multiply(
   A=Multiply(
    A=ConstantBiasScale(None=@n45.R) {bias=0.2, scale=1}
    B=TextureSample(UVs=@n47) {texture=/VehicleCosmetics/Wheels/RL251SE/Textures/Noise_Spots_RGB_T}.B
   )
   B=ScalarParameter() {defaultValue=3, parameterName=Stars_Intensity}
  )
  n49 = Multiply(
   A=VectorParameter() {parameterName=Sky_color_light, defaultValue=<rgba 0.0628 0.285 0.4844 1>}.RGB
   B=ScalarParameter() {defaultValue=1, parameterName=SkyLight_intensity}
  )
  n50 = LinearInterpolate(A=@n28.RGB, B=@n49, Alpha=LinearInterpolate(A=@n44.G, Alpha=NamedRerouteUsage()) {constB=0})
  NamedRerouteDeclaration(None=Add(A=Add(A=@n48, B=@n50), B=NamedRerouteUsage())) {name=Sky, nodeColor=<rgba 0 1 0.0156 1>}
  NamedRerouteDeclaration(None=VertexColor().R) {name=VC_R, nodeColor=<rgba 1 0 0 1>}
  n51 = NamedRerouteUsage()
  n52 = ScalarParameter() {defaultValue=1, parameterName=PaintedColorOverride}
  NamedRerouteDeclaration(None=Add(A=@n48, B=LinearInterpolate(A=@n51, B=@n49, Alpha=@n52))) {name=Sky_light, nodeColor=<rgba 0 0.2266 1 1>}
  Power(Base=@n41)
  StaticSwitchParameter(True=@n21, False=@n20) {parameterName=PaintedColorOverride}
  StaticSwitchParameter(True=@n51, False=@n49) {parameterName=PaintedColorOverride}
 ]}
